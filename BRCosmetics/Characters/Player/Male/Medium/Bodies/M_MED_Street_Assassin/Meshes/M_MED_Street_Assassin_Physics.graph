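# Physics asset for the Street Assassin medium male player body cosmetic, covering the hood and zipper dynamic bones plus the upper spine.
# Single capsules sit on dyn_hood, dyn_zipper, parent_dyn_zipper, spine_04 and spine_05.
# The hood is constrained to spine_05 and the zipper directly to spine_04, both with narrow swing and wider twist limits.
# spine_05 and parent_dyn_zipper link to spine_04 with free swing and twist.
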PhysicsAsset /BRCosmetics/Characters/Player/Male/Medium/Bodies/M_MED_Street_Assassin/Meshes/M_MED_Street_Assassin_Physics
bodies {
 dyn_hood [{shapeType=Capsule, length=8.0185, radius=7.1695, center=<0 -0.2 0.2>, rotation=<rot 0 0 -27.1057>}]
 dyn_zipper [{shapeType=Capsule, length=3, radius=0.5, shapeName=dyn_zipper_capsule, center=<0 -0.0574 -2.75>, rotation=<rot 0 0 -0.0001>}]
 parent_dyn_zipper [{shapeType=Capsule, length=2.3075, radius=1.2153, center=<-0.1036 -0.1701 -2.0912>, rotation=<rot 5.8505 -90 5.0034>}]
 spine_04 [{shapeType=Capsule, length=3.5, radius=13.5, rotation=<rot 90 -0.0001 0>}]
 spine_05 [{shapeType=Capsule, length=3.2309, radius=1.324}]
}
constraints [
 {bone1Name=dyn_hood, bone2Name=spine_05, swing1LimitDegrees=10, swing2LimitDegrees=10, twistLimitDegrees=45}
 {
  bone1Name=spine_05
  bone2Name=spine_04
  swing1LimitDegrees=45
  swing1Motion=Free
  swing2LimitDegrees=45
  swing2Motion=Free
  twistLimitDegrees=45
  twistMotion=Free
 }
 {bone1Name=dyn_zipper, bone2Name=spine_04, swing1LimitDegrees=2, swing2LimitDegrees=2, twistLimitDegrees=65}
 {
  bone1Name=parent_dyn_zipper
  bone2Name=spine_04
  swing1LimitDegrees=45
  swing1Motion=Free
  swing2LimitDegrees=45
  swing2Motion=Free
  twistLimitDegrees=45
  twistMotion=Free
 }
]
