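# Animation Blueprint for the Street Assassin medium male body cosmetic, copying the attached parent mesh's pose and adding corrective and secondary motion.
# Each update maps head-to-spine and pelvis-to-thigh yaw deltas from ref pose to clamped alphas.
# These alphas drive additive hood rotation and thigh-fix bone offsets.
# A weighted spine/clavicle constraint then positions the zipper parent bone, followed by RigidBody simulation in spine_05 space with overridden gravity.
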AnimBlueprint /BRCosmetics/Characters/Player/Male/Medium/Bodies/M_MED_Street_Assassin/Meshes/M_MED_Street_Assassin_AnimBP
target_skeleton=&/M_MED_Street_Assassin_Skeleton
graphs [
 AnimGraph {nodes [
   n1 = "Transform (Modify) Bone - Bone: dyn_hood"(
    ComponentPose=Local To Component(LocalPose=Copy Pose From Mesh() {bUseAttachedParent=true})
    Alpha=Get head_rot_back() -> (head_rot_back:double)
    Translation="0,0,0"
    Rotation="0.000,0.000,-34"
    Scale="1,1,1"
   ) {
    rotationMode=BMM_Additive
    rotationSpace=BCS_BoneSpace
    alphaBoolBlend {blendOption=Linear}
    boneToModify {boneName=dyn_hood}
    rotation {roll=-34}
   }
   n2 = "Transform (Modify) Bone - Bone: thigh_fix_l"(ComponentPose=@n1, Alpha=Get thigh_fix_l() -> (thigh_fix_l:double), Translation="0,-4,3", Rotation="0,0,0", Scale="1,1,1") {
    rotationSpace=BCS_BoneSpace
    translationMode=BMM_Additive
    translationSpace=BCS_BoneSpace
    alphaBoolBlend {blendOption=Linear}
    boneToModify {boneName=thigh_fix_l}
    translation {y=-4, z=3}
   }
   n3 = "Transform (Modify) Bone - Bone: thigh_fix_r"(ComponentPose=@n2, Alpha=Get thigh_fix_r() -> (thigh_fix_r:double), Translation="0,4,6", Rotation="0,0,0", Scale="1,1,1") {
    rotationSpace=BCS_BoneSpace
    translationMode=BMM_Additive
    translationSpace=BCS_BoneSpace
    alphaBoolBlend {blendOption=Linear}
    boneToModify {boneName=thigh_fix_r}
    translation {y=4, z=6}
   }
   n4 = "Constraint - parent_dyn_zipper (PPPPPPP)"!(
    ComponentPose=@n3
    Alpha=1
    ConstraintWeights_0=0.05
    ConstraintWeights_1:float=0.1
    ConstraintWeights_2:float=0.1
    ConstraintWeights_3:float=0.3
    ConstraintWeights_4:float=0.4
    ConstraintWeights_5:float=0
    ConstraintWeights_6:float=0.05
   ) -> (Pose) {
    alphaBoolBlend {blendOption=Linear}
    boneToModify {boneName=parent_dyn_zipper}
    constraintSetup [
     {targetBone {boneName=clavicle_l}}
     {targetBone {boneName=spine_01}}
     {targetBone {boneName=spine_02}}
     {targetBone {boneName=spine_03}}
     {targetBone {boneName=spine_04}}
     {targetBone {boneName=spine_05}}
     {targetBone {boneName=clavicle_r}}
    ]
    constraintWeights [1,1,1,1,1,1,1]
   }
   n5 = RigidBody(ComponentPose=@n4, Alpha=0.8) {
    bClampLinearTranslationLimitToRefPose=true
    bOverrideWorldGravity=true
    overridePhysicsAsset=&/M_MED_Street_Assassin_Physics
    simulationSpace=BaseBoneSpace
    alphaBoolBlend {blendOption=Linear}
    baseBoneRef {boneName=spine_05}
    componentAppliedLinearAccClamp {z=6000}
    componentLinearAccScale=<0.2 0.2 0.1>
    componentLinearVelScale=<0.2 0.2 0.1>
    overrideWorldGravity {z=-980}
   }
   Output Pose(Result=Component To Local(ComponentPose=@n5)) {layerGroup=None}
  ]}
 EventGraph {nodes [
   Event BlueprintUpdateAnimation()
   > n1 = Sequence!(execute) -> (then_0, then_1, then_2:Exec)
   n2 = BreakTransform(InTransform=GetDeltaTransformFromRefPose(self=Get Skeletal_Mesh() -> (Skeletal_Mesh:Skeletal Mesh Component Object Reference), BoneName=head, BaseName=spine_05))
   > Set head_rot_back(head_rot_back:double=MapRangeClamped(Value=BreakRotator(InRot=@n2.Rotation).Yaw, InRangeA=5, InRangeB=-42)) -> (Output_Get:double)
   n3 = BreakTransform(InTransform=GetDeltaTransformFromRefPose(self=Get Skeletal_Mesh(), BoneName=pelvis, BaseName=thigh_r))
   Set thigh_fix_r(execute=@n1.then_1, thigh_fix_r:double=MapRangeClamped(Value=BreakRotator(InRot=@n3.Rotation).Yaw, InRangeA=15, InRangeB=115)) -> (Output_Get:double)
   n4 = BreakTransform(InTransform=GetDeltaTransformFromRefPose(self=Get Skeletal_Mesh(), BoneName=pelvis, BaseName=thigh_l))
   Set thigh_fix_l(execute=@n1.then_2, thigh_fix_l:double=MapRangeClamped(Value=BreakRotator(InRot=@n4.Rotation).Yaw, InRangeA=15, InRangeB=115)) -> (Output_Get:double)
   Event BlueprintInitializeAnimation()
   n5 = GetOwnerSkeletalMeshComponent()
   > Is Valid(InputObject=@n5)
   > Set Skeletal_Mesh(Skeletal_Mesh:Skeletal Mesh Component Object Reference=@n5) -> (Output_Get:Skeletal Mesh Component Object Reference)
  ]}
]
variables [
 head_rot_back {type=number}
 Skeletal_Mesh {type=SkeletalMeshComponent}
 thigh_fix_l {type=number}
 thigh_fix_r {type=number}
]
